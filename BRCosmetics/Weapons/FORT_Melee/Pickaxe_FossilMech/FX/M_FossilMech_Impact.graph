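# Translucent unlit Niagara sprite material, named for impact effects on the FossilMech pickaxe.
# It samples a crosshair reticle mask texture with doubled U tiling, shapes it with an exponent parameter and multiplies it by particle alpha.
# Particle color is separately multiplied by a Tint parameter.
# A ShapeSelect parameter drives a three-input lerp, but all three inputs read the red channel, so it selects no different shape.
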Material /BRCosmetics/Weapons/FORT_Melee/Pickaxe_FossilMech/FX/M_FossilMech_Impact
graph {
 properties {bUsedWithNiagaraSprites=true, blendMode=BLEND_Translucent, pixelDepthOffsetMode=PDOM_Legacy, shadingModel=MSM_Unlit, usedShadingModels=Unlit}
 nodes [
  n1 = TextureSampleParameter2D(UVs=TextureCoordinate() {uTiling=2}) {parameterName=MainTex, samplerType=SAMPLERTYPE_Masks, texture=&/T_CrosshairReticle_FX}
  n2 = ScalarParameter() {parameterName=ShapeSelect}
  n3 = ParticleColor()
  Multiply(
   A=Power(
    Base=MaterialFunctionCall!(1=@n1.R, 2=@n1.R, 3=@n1.R, 4, A=@n2) {materialFunction=/Engine/Functions/Engine_MaterialFunctions02/Lerp_Multiple_Float}.Lerp 3 Inputs
    Exp=ScalarParameter() {defaultValue=1, parameterName=Exp}
   )
   B=@n3.A
  )
  Multiply(A=@n3.RGB, B=VectorParameter() {parameterName=Tint, defaultValue=<rgba 1 1 1 0>}.RGB)
 ]
}
