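# Translucent, unlit, two-sided material for the FossilMech pickaxe FX, enabled for Niagara ribbons, drawing a glowing circuit-board streak with a hard edge line.
# Offset circuit samples form a color-fringed pattern, modulated by scrolling flame and layered with gradient, glow and edge colors.
# Opacity combines animated digital noise, UV masks, the edge line, depth fade and particle alpha.
# Dynamic parameters drive opacity, brightness and hue shift.
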Material /BRCosmetics/Weapons/FORT_Melee/Pickaxe_FossilMech/FX/M_FossilMecha_Pickaxe
graph {
 properties {
  bUsedWithNiagaraRibbons=true
  blendMode=BLEND_Translucent
  pixelDepthOffsetMode=PDOM_Legacy
  shadingModel=MSM_Unlit
  twoSided=true
  usedShadingModels=Unlit
 }
 nodes [
  LinearInterpolate()
  n1 = DynamicParameter() {defaultValue=<rgba 1 1 0 1>, paramNames [Param1,Param2,HueShift,Param4]}
  n2 = VectorParameter() {parameterName=UV, defaultValue=<rgba 2.5 0.4 1.4 0.6>}
  n3 = MaterialFunctionCall(XY=Multiply(A=TextureCoordinate(), B=AppendVector(A=@n2.R, B=@n2.R))) {materialFunction=&Swizzle}
  n4 = Constant() {r=0.98}
  n5 = Constant() {r=1.02}
  n6 = MaterialFunctionCall(
   X=TextureSample(UVs=Panner(Coordinate=@n3.YX, Time=@n4) {speedX=0.1, speedY=0.1}) {texture=/Game/Effects/Fort_Effects/Textures/T_FX_Circuitboard}.R
   Y=TextureSample(UVs=@n3.YX) {texture=&T_FX_Circuitboard}.R
   Z=TextureSample(UVs=Panner(Coordinate=@n3.YX, Time=@n5) {speedX=0.1, speedY=0.1}) {texture=&T_FX_Circuitboard}.R
  ) {materialFunction=&MakeFloat3}
  n7 = TextureSample(UVs=Panner(Coordinate=TextureCoordinate() {uTiling=2}) {speedY=-0.3}) {samplerType=SAMPLERTYPE_LinearGrayscale, texture=/BRCosmetics/Weapons/FORT_Melee/Pickaxe_Cube_Ninja_Male/FX/T_FluidFlameTile}
  n8 = LinearInterpolate(
   A=Multiply(A=ScalarParameter() {defaultValue=20, parameterName=CircuitMultiply}, B=Multiply(A=@n6.Result, B=@n7.R))
   B=Multiply(A=Desaturation(None=@n6.Result), B=VectorParameter() {parameterName=BoardColor, defaultValue=<rgba 1 1 1 0>}.RGB)
   Alpha=ScalarParameter() {parameterName=UseSingleBoardColor}
  )
  n9 = TextureCoordinate()
  n10 = LinearInterpolate(
   A=VectorParameter() {parameterName=ColorA, defaultValue=<rgba 8 0.2321 1.431 1>}.RGB
   B=VectorParameter() {parameterName=ColorB, defaultValue=<rgba 0.8742 0 1 1>}.RGB
   Alpha=Power(Base=ComponentMask(None=@n9) {g=true}) {constExponent=1}
  )
  n11 = VectorParameter() {parameterName=Glow, defaultValue=<rgba 1 0.1354 0.5628 1>}
  n12 = TextureCoordinate()
  n13 = Power(Base=OneMinus(None=ComponentMask(None=@n12) {r=true}))
  n14 = Constant() {r=0.01}
  n15 = AppendVector(A=@n14, B=ComponentMask(None=TextureCoordinate()) {g=true})
  n16 = ComponentMask(None=TextureCoordinate()) {r=true}
  n17 = LinearInterpolate(
   A=ScalarParameter() {defaultValue=0.05, group=Hard Line, parameterName=Edge RadiusAltA, sortPriority=1}
   B=ScalarParameter() {defaultValue=0.01, group=Hard Line, parameterName=Edge RadiusAltB, sortPriority=2}
   Alpha=@n16
  )
  n18 = LinearInterpolate(
   A=ScalarParameter() {defaultValue=0.3, group=Hard Line, parameterName=Edge Hardness A, sortPriority=3}
   B=ScalarParameter() {group=Hard Line, parameterName=Edge Hardness B, sortPriority=4}
   Alpha=@n16
  )
  n19 = LinearInterpolate(
   A=ScalarParameter() {defaultValue=0.03, group=Hard Line, parameterName=Edge Radius A, sortPriority=1}
   B=ScalarParameter() {group=Hard Line, parameterName=Edge Radius B, sortPriority=2}
   Alpha=@n16
  )
  n20 = Multiply(A=LinearInterpolate(
    A=SphereMask(A=@n15, B=@n14, Radius=@n17, Hardness=@n18)
    B=SphereMask(A=@n15, B=@n14, Radius=@n19, Hardness=@n18)
    Alpha=ScalarParameter() {defaultValue=1, parameterName=UseAltLine}
   ))
  n21 = LinearInterpolate(
   A=Add(A=Add(A=@n8, B=@n10), B=Multiply(A=@n11.RGB, B=@n13))
   B=VectorParameter() {parameterName=EdgeColor, defaultValue=<rgba 55 0 17.2443 1>}.RGB
   Alpha=@n20
  )
  MaterialFunctionCall(Emissive=MaterialFunctionCall(Hue Shift Percentage=@n1.HueShift, Texture=Multiply(A=@n21, B=@n1.Param2)) {materialFunction=&HueShift}.Result) {materialFunction=&MF_FX_EmissiveAdjustments}
  n22 = AppendVector(A=@n2.B, B=@n2.B)
  n23 = TextureCoordinate()
  n24 = TextureSample(UVs=Panner(Coordinate=Multiply(A=@n22, B=@n23)) {speedX=1}) {samplerType=SAMPLERTYPE_LinearColor, texture=/Game/Global/Textures/Default/Noise/T_Atlas_DigitalNoise_01}
  n25 = MaterialFunctionCall() {materialFunction=&DebugTimeSine}
  n26 = Constant() {r=0.5}
  n27 = TextureCoordinate()
  n28 = OneMinus(None=ComponentMask(None=@n27) {g=true})
  n29 = SphereMask(A=ComponentMask(None=@n27) {r=true}, B=Constant() {r=-0.1}) {attenuationRadius=1, hardnessPercent=0}
  n30 = Multiply(A=Multiply(A=@n28, B=@n28), B=@n29)
  n31 = Add(
   A=LinearInterpolate(
    A=LinearInterpolate(A=@n24.R, B=@n24.R, Alpha=@n25.Result)
    B=@n24.B
    Alpha=MaterialFunctionCall(Speed=@n26) {materialFunction=&DebugTimeSine}.Result
   )
   B=@n30
  )
  n32 = TextureCoordinate() {uTiling=0.3}
  n33 = Constant2Vector() {g=0.5, r=1}
  n34 = OneMinus(None=Multiply(A=SphereMask(A=@n32, B=@n33) {attenuationRadius=1, hardnessPercent=0}) {constB=5})
  n35 = Multiply(A=Multiply(A=Power(Base=@n34)) {constB=0.55}, B=@n30)
  n36 = ComponentMask(None=TextureCoordinate()) {r=true}
  n37 = ScalarParameter() {parameterName=Gradient_Offset}
  n38 = Power(
   Base=SphereMask(
    A=Add(A=@n36, B=@n37)
    B=Constant() {r=1}
    Radius=ScalarParameter() {defaultValue=1, parameterName=Gradient_Radius}
    Hardness=ScalarParameter() {defaultValue=40, parameterName=Gradient_Hardness}
   ) {attenuationRadius=1, hardnessPercent=40}
   Exp=ScalarParameter() {defaultValue=2.5, parameterName=Gradient_Pow}
  ) {constExponent=2.5}
  n39 = ScalarParameter() {defaultValue=6, parameterName=GradientMultiply}
  n40 = Multiply(A=Add(A=Multiply(A=@n31, B=@n35), B=@n20), B=Saturate(None=Multiply(A=@n38, B=@n39) {constB=6}))
  n41 = ScalarParameter() {defaultValue=10, parameterName=DepthFade}
  Multiply(
   A=ParticleColor().A
   B=Multiply(
    A=DepthFade(Opacity=@n40, FadeDistance=@n41)
    B=DynamicParameter() {defaultValue=<rgba 1 1 0 1>, paramNames [Param1,Param2,HueShift,Param4]}.Param1
   )
  )
  Multiply(A=@n29, B=@n29)
 ]
}
