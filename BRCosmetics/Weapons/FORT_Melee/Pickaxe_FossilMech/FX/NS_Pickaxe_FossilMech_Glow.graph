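# Glow effect for the FossilMech pickaxe that shifts from blue to orange as System.WeakPointTimer rises.
# LightAndGlow bursts two particles at fixed local offsets, drawn with sprite, mesh and two light renderers.
# The timer blends colors and the second light's pulse, while time-of-day scalars scale brightness and a camera-vector cross product scales glow alpha.
# The owner's Y axis feeds dynamic material parameters; disabled FN_Debug passes the timer to a material.
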NiagaraSystem /BRCosmetics/Weapons/FORT_Melee/Pickaxe_FossilMech/FX/NS_Pickaxe_FossilMech_Glow
dependencies {
 usedDataInterfaces [&NiagaraDataInterfaceArrayFloat3, &NiagaraDataInterfaceCamera, &NiagaraDataInterfaceCurve]
 usedDynamicInputs [
  &FloatFromCurve
  &OneMinusFloat
  &ClampFloat
  &Divide_Float
  &Subtract_Float
 ]
 usedModules [
  &SystemState
  &self:SystemSpawnScript.SystemScriptSource.SystemScriptGraph.NiagaraNodeAssignment_0.SetVariables_#1
  &self:SystemSpawnScript.SystemScriptSource.SystemScriptGraph.NiagaraNodeAssignment_1.SetVariables_#2
 ]
}
emitters {
 FN_Debug {
  bEnabled=false
  emitterUpdateScript [
   &EmitterState()
   &SpawnBurst_Instantaneous(Spawn Count=1, Use Spawn Probability=false)
  ]
  particleSpawnScript [/Niagara/Modules/Spawn/Initialization/InitializeParticle(
    Lifetime=5
    Module.Write Color=true
    Module.Write Lifetime=true
    Module.Write Mass=true
    Module.Write Material Random=true
    Module.Write Position=false
    Module.Write Scale=false
    Module.Write SpriteRotation=false
    Module.Write SpriteSize=true
    Position=&SimulationPosition()
    Sprite Size=<100 100>
   )]
  particleUpdateScript [
   &ParticleState()
   /SetVariables(Particles.Debug:NiagaraFloat=$System.WeakPointTimer) {isSetParameters=true}
   &DynamicMaterialParameters(
    Index 0 Param 1=$Particles.Debug
    Module.Param0WriteEnabled=true
    Module.Param1WriteEnabled=true
    Module.Param2WriteEnabled=true
    Module.Param3WriteEnabled=true
   )
  ]
  rendererClasses [&NiagaraSpriteRendererProperties]
  renderers [{rendererClass=&NiagaraSpriteRendererProperties}]
 }
 LightAndGlow {
  emitterUpdateScript [
   &EmitterState()
   &SpawnBurst_Instantaneous(Spawn Count=2, Use Spawn Probability=false, UseLoopCountLimit=true)
  ]
  particleSpawnScript [
   /Niagara/Modules/Spawn/Initialization/V2/InitializeParticle(
    Color=&LerpLinearColors(StartColor=<rgba 0.528 2.5177 4 2>, EndColor=<rgba 4 1.9 0.5 4>, LerpFactor=$System.WeakPointTimer)
    Lifetime=5
    Position Offset=&SelectVectorFromArray(
     Direct Array Index=&ReturnExecIndex()
     Vector Selection Array=DataInterface() {propertyValues {FloatData [<0 38 52.5>, <0 -22 52.5>]}}
    )
    Position Offset Coordinate Space=::Local
    Sprite Size Mode=::NewEnumerator3
    Uniform Sprite Size=50
    UsePositionOffset=true
    Write Material Random=true
   )
   &CameraOffset(Camera Offset Amount=-30)
  ]
  particleUpdateScript [
   &ParticleState(Kill Particles When Lifetime Has Elapsed=false)
   &DynamicMaterialParameters(
    Index 0 Param 1=&MakeFloatFromVector(Vector=$Engine.Owner.SystemYAxis, Channel=::NewEnumerator0)
    Index 0 Param 2=&MakeFloatFromVector(Vector=$Engine.Owner.SystemYAxis, Channel=::NewEnumerator1)
    Index 0 Param 3=&MakeFloatFromVector(Vector=$Engine.Owner.SystemYAxis, Channel=::NewEnumerator2)
    Index 0 Param 4=1
    Module.Param0WriteEnabled=true
    Module.Param1WriteEnabled=true
    Module.Param2WriteEnabled=true
    Module.Param3WriteEnabled=true
   )
   &FN_Light_Attributes(
    Color=&LerpLinearColors(StartColor=<rgba 1.1453 33.2948 80 1>, EndColor=<rgba 80 18.2305 1.1453 1>, LerpFactor=$System.WeakPointTimer)
    Create Color Attribute=true
    Create Exponent Attribute=true
    Create Position Attribute=true
    Create Radius Attribute=true
    Exponent=40
    Position=&AddVectorToPosition(
     Position=$Particles.Position
     Vector=&Multiply_VectorByFloat(Vector=<0 2 0>, Float=&Lerp_Float(A=1, B=-1, Alpha=&ReturnNormalizedExecIndex()))
    )
    Radius=60
    Scale Color=&Multiply_Float(
     A=&Multiply_Float(
      A=&Multiply_Float(A=2, B=&FNTimeOfDayScalar(Night Scale=0.2, Morning Scale=1, Day Scale=1.3, Evening Scale=0.3))
      B=&Sine(Normalized Angle=$Emitter.Age, Period=0.5, Scale=0.15, Bias=1)
     )
     B=&FNTimeOfDayScalar(Night Scale=0.4, Morning Scale=0.5, Day Scale=0.3, Evening Scale=0.5)
    )
   )
   &FN_Light_Attributes(
    Color=&LerpLinearColors(StartColor=<rgba 1.1453 33.2948 80 1>, EndColor=<rgba 80 30.0587 1.1453 1>, LerpFactor=$System.WeakPointTimer)
    Create Color Attribute=true
    Create Exponent Attribute=true
    Create Position Attribute=true
    Create Radius Attribute=true
    Exponent=12
    Position=&AddVectorToPosition(
     Position=$Particles.Position
     Vector=&Multiply_VectorByFloat(Vector=<0 2 0>, Float=&Lerp_Float(A=1, B=-1, Alpha=&ReturnNormalizedExecIndex()))
    )
    Radius=70
    Scale Color=&Multiply_Float(
     A=&Lerp_Float(A=1, B=&Sine(Normalized Angle=$Emitter.Age, Period=0.4, Scale=0.1, Bias=0.7), Alpha=$System.WeakPointTimer)
     B=&FNTimeOfDayScalar(Night Scale=0.3, Morning Scale=0.5, Day Scale=0.3, Evening Scale=0.4)
    )
   ) {module=FN_Light_Attributes001}
   &Color(Color=&LerpLinearColors(StartColor=<rgba 0.528 2.5177 4 2>, EndColor=<rgba 4 1.1111 0.312 4>, LerpFactor=$System.WeakPointTimer))
   &ScaleColor(
    Color Value To Scale=$Particles.Color
    Scale Alpha=&Multiply_Float(
     A=&Multiply_Float(
      A=1.2
      B=&FNTimeOfDayScalar(Night Scale=0.07, Morning Scale=0.1, Day Scale=0.07, Evening Scale=0.1)
     )
     B=&VectorLength(
      Vector=/Niagara/DynamicInputs/Vectors/CrossProduct(
       Vector A:Vector3f=&ReturnCameraVector(Camera Vectors=::NewEnumerator1, Coordinate Space=::Simulation)
       Vector B:Vector3f=<1 0 1>
       Normalize Incoming Vectors:NiagaraBool=true
      )
      Fallback Vector=<0 -1 0>
     )
    )
    ScaleA=true
    ScaleRGB=true
   )
  ]
  rendererClasses [
   &NiagaraSpriteRendererProperties
   &NiagaraLightRendererProperties
   &NiagaraMeshRendererProperties
  ]
  renderers [
   {rendererClass=&NiagaraSpriteRendererProperties}
   {rendererClass=&NiagaraLightRendererProperties, rendererIndex=1}
   {rendererClass=&NiagaraMeshRendererProperties, rendererIndex=2}
   {rendererClass=&NiagaraLightRendererProperties, rendererIndex=3}
  ]
 }
}
summary {
 systemName=NS_Pickaxe_FossilMech_Glow
 emitters [
  {
   emitterName=LightAndGlow
   rendererClasses [
    &NiagaraSpriteRendererProperties
    &NiagaraLightRendererProperties
    &NiagaraMeshRendererProperties
   ]
  }
  {bEnabled=false, emitterName=FN_Debug, rendererClasses [&NiagaraSpriteRendererProperties]}
 ]
 userVariables [
  User.bIsFrontEnd {type=NiagaraBool}
  User.bIsFrontEndPreview {type=NiagaraBool}
  User.bisSwinging {type=NiagaraBool}
  User.Debug {type=NiagaraFloat, defaultValue=0, description=""}
  User.PlayerSpeed {type=NiagaraFloat, defaultValue=0, description=""}
  User.RibbonIDOffset {type=NiagaraInt32}
  User.SkeletalMesh {
   type=NiagaraDataInterfaceSkeletalMesh
   defaultValue {
    struct=NiagaraExt_VariableValue_DataInterface
    value {dataInterface=&self:NiagaraDataInterfaceSkeletalMesh_2}
   }
  }
  User.WeakPointTime {type=NiagaraFloat, defaultValue=0, description=""}
 ]
}
