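# Material function for the FossilMech pickaxe that builds an animated emissive with moving highlight bands and also outputs a tech-art mask.
# UVs pan in opposite directions, apparently split at a MiddleAxis threshold, then are tiled, offset by mask values and wrapped with Frac.
# A SphereMask turns that repeating ramp into soft highlights blending background and highlight colors.
# Named reroutes cache channels of the input tech-art mask and original emissive for offsets and masking.
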MaterialFunction /BRCosmetics/Weapons/FORT_Melee/Pickaxe_FossilMech/Materials/MF_Pickaxe_FossilMech_AdvancedEmissive
graph {nodes [
  n1 = Constant3Vector() {constant=<rgba 0.155 0.5724 1 1>}
  n2 = Constant3Vector() {constant=<rgba 0.8854 1 1 1>}
  n3 = AppendVector(A=ScalarParameter() {defaultValue=0.02, group=MF_AdvancedEM, parameterName=XPannerSpeed}, B=Constant())
  n4 = Subtract(B=@n3) {constA=0}
  n5 = Multiply(
   A=LinearInterpolate(A=Panner(Speed=@n3), B=Panner(Speed=@n4), Alpha=NamedRerouteUsage())
   B=ScalarParameter() {defaultValue=25, group=MF_AdvancedEM, parameterName=EM_Tiling}
  )
  n6 = NamedRerouteUsage()
  n7 = ScalarParameter() {defaultValue=3, group=MF_AdvancedEM, parameterName=EM_Offset_Multi}
  n8 = Multiply(A=@n6, B=@n7)
  n9 = Multiply(A=@n6, B=Divide(A=@n7))
  n10 = NamedRerouteUsage()
  n11 = Frac(None=Add(A=ComponentMask(None=@n5) {r=true}, B=LinearInterpolate(A=@n8, B=@n9, Alpha=@n10)))
  n12 = Multiply(A=SphereMask(
    A=Abs(None=@n11)
    B=ScalarParameter() {defaultValue=0.25, group=MF_AdvancedEM, parameterName=EM_highlight}
    Radius=ScalarParameter() {defaultValue=0.15, group=MF_AdvancedEM, parameterName=EM_highlight_radius}
   ) {hardnessPercent=0}) {constB=2}
  n13 = LinearInterpolate(
   A=Multiply(
    A=FunctionInput(Preview=@n1) {bUsePreviewValueAsDefault=true, inputName=EmissiveInput01}
    B=ScalarParameter() {defaultValue=1.5, group=MF_AdvancedEM, parameterName=BackGroundEmissiveIntensity}
   )
   B=Multiply(
    A=FunctionInput(Preview=@n2) {bUsePreviewValueAsDefault=true, inputName=EmissiveInput02}
    B=ScalarParameter() {defaultValue=3, group=MF_AdvancedEM, parameterName=HighLightEmissiveIntensity}
   )
   Alpha=Multiply(A=Saturate(None=@n12), B=NamedRerouteUsage())
  )
  n14 = NamedRerouteUsage()
  FunctionOutput(None=Multiply(A=Multiply(A=@n13, B=@n14), B=NamedRerouteUsage())) {outputName=AdvancedEmissive_Fossil}
  FunctionOutput(None=NamedRerouteUsage()) {outputName=AdditionalTechArtMask, sortPriority=1}
  n15 = TextureCoordinate()
  n16 = NamedRerouteDeclaration(None=ScalarParameter() {defaultValue=0.277, group=MF_AdvancedEM, parameterName=MiddleAxis}) {name=CachedMiddleAxis, nodeColor=<rgba 0 1 0.7656 1>}
  n17 = Constant()
  NamedRerouteDeclaration(None=If(A=Subtract(A=@n15, B=@n16), "A > B"=Constant() {r=1}, "A == B"=@n17, "A < B"=@n17)) {name=CachedAxisMask, nodeColor=<rgba 0.875 0 1 1>}
  n18 = ComponentMask(None=FunctionInput() {bUsePreviewValueAsDefault=true, inputName=OrginalTechArtMask}) {b=true, g=true, r=true}
  NamedRerouteDeclaration(None=@n18) {name=CachedAdditionalMask, nodeColor=<rgba 0.4766 1 0 1>}
  NamedRerouteDeclaration(None=ComponentMask(None=@n18) {b=true}) {name=CachedGlassAreaMask, nodeColor=<rgba 0.9453 1 0 1>}
  n19 = NamedRerouteDeclaration(None=ComponentMask(None=@n18) {g=true}) {name=CachedLineVarientMask, nodeColor=<rgba 1 0.7734 0 1>}
  NamedRerouteDeclaration(None=Saturate(None=Multiply(A=@n19) {constB=100})) {name=CachedGradientMask, nodeColor=<rgba 1 0.9609 0 1>}
  NamedRerouteDeclaration(None=FunctionInput() {bUsePreviewValueAsDefault=true, inputName=OrginalEmissiveInput}) {name=CachedEmissiveMaskValue, nodeColor=<rgba 1 0.8906 0 1>}
 ]}
