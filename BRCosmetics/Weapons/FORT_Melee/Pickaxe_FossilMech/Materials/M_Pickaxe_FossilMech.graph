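# Master material for the FossilMech pickaxe, built on the shared character PreFX, SkinFX, PostFX and cinematic PostFX function chain.
# After SkinFX, a custom advanced-emissive function blends back/front colors by tech-art mask, with a switch for original emission.
# A tinted, optionally cropped warning-signal texture is lerped over that emissive before PostFX.
# Named reroutes cache glass masks, sine-driven pulses and a timed, reversible weak-state multiplier broken up by panning glitch noise.
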Material /BRCosmetics/Weapons/FORT_Melee/Pickaxe_FossilMech/Materials/M_Pickaxe_FossilMech
graph {
 properties {
  bAutomaticallySetUsageInEditor=false
  bUseMaterialAttributes=true
  bUsedWithClothing=true
  bUsedWithMeshDeformer=true
  bUsedWithMorphTargets=true
  bUsedWithNanite=true
  bUsedWithSkeletalMesh=true
  pixelDepthOffsetMode=PDOM_Legacy
  refractionMethod=RM_IndexOfRefraction
  shadingModel=MSM_FromMaterialExpression
  subsurfaceProfile=&SS_HeroSkin_02
  usedShadingModels=Default Lit | Subsurface | Clear Coat | Subsurface Profile | Cloth | Eye
 }
 nodes [
  n1 = MaterialFunctionCall() {materialFunction=&BaseFN_CharacterFunction_PreFX}
  n2 = MaterialFunctionCall(
   PreskinnedPosition=@n1.Pre skinned Local Position
   Gmap=@n1.Pre skinned Local Position
   MaterialAttributes=@n1.Pre skinned Local Position
   AO=@n1.Pre skinned Local Position
  ) {materialFunction=&CharacterSkinFX}
  n3 = GetMaterialAttributes(None=@n2.Result) {attributeGetTypes [#1, #2, #3]}
  n4 = Saturate(None=Multiply(A=NamedRerouteUsage()) {constB=10})
  n5 = MaterialFunctionCall!(
   OrginalEmissiveInput=@n4
   EmissiveInput01=VectorParameter() {group=MF_AdvancedEM, parameterName=MF_EmissiveBack, defaultValue=<rgba 0.1018 0.6397 0.724 1>}.RGB
   EmissiveInput02=VectorParameter() {group=MF_AdvancedEM, parameterName=MF_EmissiveFront, defaultValue=<rgba 1 1 1 1>}.RGB
   OrginalTechArtMask=NamedRerouteUsage()
  ) {materialFunction=&/MF_Pickaxe_FossilMech_AdvancedEmissive}
  n6 = ComponentMask(None=@n5.AdditionalTechArtMask) {g=true}
  n7 = Saturate(None=Multiply(A=@n6) {constB=100})
  n8 = ComponentMask(None=@n4) {b=true}
  n9 = StaticSwitchParameter(True=NamedRerouteUsage(), False=LinearInterpolate(A=NamedRerouteUsage(), B=@n5.AdvancedEmissive_Fossil, Alpha=Multiply(A=@n8, B=@n7))) {group=MF_AdvancedEM, parameterName=UseOrginalEmission}
  n10 = TextureObjectParameter() {
   group=MF_AdvancedEM
   parameterName=Warning_Signal_Texture
   texture=/BRCosmetics/Weapons/FORT_Melee/Pickaxe_FossilMech/Textures/T_Pickaxe_FossilMech_WarningMask
  }
  n11 = VectorParameter() {group=Emissive, parameterName=EmissiveUVs_RG_UpperLeftCorner_BA_LowerRightCorner, defaultValue=<rgba 0 0 1 1>}
  n12 = StaticSwitchParameter(
   True=MaterialFunctionCall!(
    UVs
    TextureIn=@n10
    UpperLeftCorner=ComponentMask(None=@n11.RGB) {g=true, r=true}
    LowerRightCorner=AppendVector(A=@n11.B, B=@n11.B)
    ExportFloat4
   ) {materialFunction=/Engine/Functions/Engine_MaterialFunctions01/Texturing/TextureCropping}.Cropped
   False=TextureSample(Tex=@n10) {texture=&T_Pickaxe_FossilMech_WarningMask}.RGB
  ) {desc=CROP_EMIS, group=Emissive, parameterName=CroppedEmissive}
  n13 = VectorParameter() {group=MF_AdvancedEM, parameterName=WarningSiganlColor, defaultValue=<rgba 1 0 0 1>}
  n14 = Multiply(A=Multiply(A=Multiply(A=@n12, B=@n13.RGB), B=NamedRerouteUsage()), B=NamedRerouteUsage())
  n15 = NamedRerouteUsage()
  n16 = MaterialFunctionCall(
   AO=@n1.AO
   Distance Blend=@n1.AO
   Fresnel=@n1.AO
   Gmap=@n1.AO
   Pre Skinned Local Position=@n1.AO
   Material=SetMaterialAttributes!(MaterialAttributes=@n3.MaterialAttributes, Emissive Color=LinearInterpolate(A=@n9, B=@n14, Alpha=@n15), Base Color=NamedRerouteUsage()) {attributeSetTypes [#1, #3]}
  ) {materialFunction=&BaseFN_CharacterFunction_PostFX}
  MaterialFunctionCall(In=@n16.Result) {materialFunction=&CINE_MF_CharPostFX}
  n17 = NamedRerouteUsage()
  MaterialFunctionCall(Normal=@n17, Flatness=Constant() {r=-2}) {materialFunction=&FlattenNormal}
  n18 = Fresnel()
  n19 = NamedRerouteUsage()
  Multiply(A=Power(Base=Multiply(A=@n18, B=@n19)) {constExponent=1.5}) {constB=3}
  n20 = Time()
  n21 = Multiply(A=@n20)
  n22 = LinearInterpolate(A=ScalarParameter() {defaultValue=0.1, group=MF_AdvancedEM, parameterName=BaseEmissiveIntensity_02}, Alpha=Abs(None=Sine(None=@n21)))
  Multiply(A=@n22) {constB=3}
  NamedRerouteDeclaration(None=@n3.Emissive Color) {name=CachedEmissiveColorAfterSkin, nodeColor=<rgba 1 0 0.4922 1>}
  NamedRerouteDeclaration(None=@n3.Normal) {name=CachedSkinFXNormal, nodeColor=<rgba 0.3364 1 0.0104 1>}
  n23 = ComponentMask(None=@n5.AdditionalTechArtMask) {b=true}
  n24 = Saturate(None=Multiply(A=@n23) {constB=10})
  NamedRerouteDeclaration(None=@n24) {name=CachedNormalMaskForGlass, nodeColor=<rgba 1 0.8438 0 1>}
  NamedRerouteDeclaration(None=Multiply(A=@n24, B=Step(Y=@n23) {constX=0.5})) {name=CachedMiddleGlassMask, nodeColor=<rgba 0.5 1 0 1>}
  n25 = OneMinus(None=TextureSample(UVs=Panner() {speedY=3}) {samplerType=SAMPLERTYPE_LinearGrayscale, texture=/Game/Global/Textures/Default/Noise/glitch_noise}.R)
  n26 = Time()
  n27 = ScalarParameter() {group=MF_AdvancedEM, parameterName=Weak_End_Time}
  n28 = Saturate(None=Divide(A=Subtract(A=@n26, B=@n27), B=ScalarParameter() {defaultValue=6, group=MF_AdvancedEM, parameterName=Weak_Length}))
  n29 = Saturate(None=Multiply(A=Sine(None=@n28)) {constB=2})
  n30 = StaticSwitchParameter(True=OneMinus(None=@n29), False=@n29) {parameterName=ReverseProgress}
  NamedRerouteDeclaration(None=Multiply(A=Clamp(None=@n25, Min=@n30), B=Ceil(None=@n30))) {name=Cached_Weak_Time_Multipiler, nodeColor=<rgba 0 1 0.6953 1>}
  n31 = NamedRerouteUsage()
  n32 = Multiply(A=@n20, B=ScalarParameter() {defaultValue=0.3, group=MF_AdvancedEM, parameterName=MF_TimeMulti})
  n33 = LinearInterpolate(A=ScalarParameter() {defaultValue=0.3, group=MF_AdvancedEM, parameterName=BaseEmissiveIntensity_01}, Alpha=Abs(None=Sine(None=@n32)))
  NamedRerouteDeclaration(None=LinearInterpolate(A=@n31, B=Multiply(A=@n33, B=@n31), Alpha=NamedRerouteUsage())) {name=CachedTimeDrivenEmissive, nodeColor=<rgba 1 0.1172 0 1>}
  NamedRerouteDeclaration(None=@n22) {name=CachedTimeDriver, nodeColor=<rgba 1 0.4219 0 1>}
  NamedRerouteDeclaration(None=@n2.SkinFX_Mask) {name=CachedTechArtMask, nodeColor=<rgba 1 0 0.7266 1>}
  NamedRerouteDeclaration(None=@n3.Base Color) {name=CachedBaseColorAfterSkin, nodeColor=<rgba 1 0 0.9141 1>}
  n34 = ScalarParameter() {defaultValue=0.1, group=MF_AdvancedEM, parameterName=MF_GradientMin}
  SmoothStep(Min=@n34, Max=Multiply(A=@n34) {constB=2}, Value=@n6) {constMin=0.005}
  Transform(None=@n17)
 ]
}
